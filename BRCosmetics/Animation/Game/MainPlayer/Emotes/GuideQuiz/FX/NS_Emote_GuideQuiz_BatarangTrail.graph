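# Niagara effect for the GuideQuiz emote: a red batarang ribbon trail with dark smoke.
# BatarangTrail uses FN anim-trail modules between two bones or sockets, tinting its short-lived ribbons by time of day.
# At socket 2, Smoke bursts sprites in two spawn groups, one flung along preset velocities with drag, while SmokeTrail emits a brief fading ribbon.
# BatarangTrail001, a socket-sampled spawn-rate version of the trail, is disabled.
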NiagaraSystem /BRCosmetics/Animation/Game/MainPlayer/Emotes/GuideQuiz/FX/NS_Emote_GuideQuiz_BatarangTrail
dependencies {
 usedDataInterfaces [
  &NiagaraDataInterfaceCurve
  &NiagaraDataInterfaceArrayFloat3
  &NiagaraDataInterfaceArrayFloat
  &NiagaraDataInterfaceSocketReader
 ]
 usedDynamicInputs []
 usedModules [
  &self:SystemSpawnScript.SystemScriptSource.SystemScriptGraph.NiagaraNodeAssignment_0.SetVariables_#1
  &SystemState
 ]
}
emitters {
 BatarangTrail {
  emitterUpdateScript [
   &EmitterState()
   &FN_SpawnAnimTrail()
   &SpawnRate(SpawnRate=160, Use Spawn Probability=false) {enabled=false}
  ]
  particleSpawnScript [
   &FN_InitializeAnimTrailBoneIndecies(First Bone or Socket Index=1, Second Bone or Socket Index=0)
   &FN_InitializeAnimTrail(AnimTrailWidth=1.1, Lifetime=0.2)
   &InitializeParticle(
    Color=<rgba 1 0 0 1>
    Lifetime=0.075
    Ribbon Width=15
    Ribbon Width Mode=::NewEnumerator1
    UsePositionOffset=false
    Write Material Random=true
   ) {enabled=false}
   &SocketLocation(SamplePosOffset=false, Socket Reader=$System.Socket Reader, Socket Sampling Mode=::NewEnumerator3, VelocitySpeedLimit=false) {enabled=false}
   &AddVelocity(Velocity=<500 0 0>) {enabled=false}
  ]
  particleUpdateScript [
   &ParticleState()
   &SolveForcesAndVelocity(Clamp Velocity=false, Limit Acceleration=false)
   &FN_UpdateAnimTrail()
   &ScaleRibbonWidth(Ribbon Width Scale=&FloatFromCurve(
     FloatCurve=DataInterface() {propertyValues {Curve {defaultValue=1, keys [{tangentMode=RCTM_User, time=0, value=1}, {tangentMode=RCTM_User, time=1, value=1.5}]}}}
     CurveIndex=$Engine.System.Age
    )) {enabled=false}
   &ScaleColor(
    Scale RGB=&MakeVectorFromLinearColorRGB(Color=&FNTimeOfDayColor(
      Night Color=<rgba 0.3177 0.0132 0 1>
      Morning Color=<rgba 0.5 0.059 0.0006 1>
      Day Color=<rgba 1 0 0 1>
      Evening Color=<rgba 0.6875 0.0286 0 1>
     ))
    ScaleA=true
    ScaleRGB=true
   )
  ]
  rendererClasses [&NiagaraRibbonRendererProperties]
  renderers [
   {rendererClass=&NiagaraRibbonRendererProperties}
   {rendererClass=&NiagaraRibbonRendererProperties, rendererIndex=1}
   {rendererClass=&NiagaraRibbonRendererProperties, rendererIndex=2}
  ]
 }
 BatarangTrail001 {
  bEnabled=false
  emitterUpdateScript [
   &EmitterState()
   &SpawnRate(SpawnRate=160, Use Spawn Probability=false)
  ]
  particleSpawnScript [
   &InitializeParticle(
    Color=<rgba 1 0 0 1>
    Lifetime=0.075
    Ribbon Width=15
    Ribbon Width Mode=::NewEnumerator1
    UsePositionOffset=false
    Write Material Random=true
   )
   &SocketLocation(SamplePosOffset=false, Socket Reader=$System.Socket Reader, Socket Sampling Mode=::NewEnumerator3, VelocitySpeedLimit=false)
   &AddVelocity(Velocity=<500 0 0>) {enabled=false}
  ]
  particleUpdateScript [
   &ParticleState()
   &SolveForcesAndVelocity(Clamp Velocity=false, Limit Acceleration=false)
   &ScaleRibbonWidth(Ribbon Width Scale=&FloatFromCurve(
     FloatCurve=DataInterface() {propertyValues {Curve {defaultValue=1, keys [{tangentMode=RCTM_User, time=0, value=1}, {tangentMode=RCTM_User, time=1, value=1.5}]}}}
     CurveIndex=$Engine.System.Age
    ))
  ]
  rendererClasses [&NiagaraRibbonRendererProperties]
  renderers [
   {rendererClass=&NiagaraRibbonRendererProperties}
   {rendererClass=&NiagaraRibbonRendererProperties, rendererIndex=1}
   {rendererClass=&NiagaraRibbonRendererProperties, rendererIndex=2}
  ]
 }
 Smoke {
  emitterUpdateScript [
   &EmitterState(Life Cycle Mode=::NewEnumerator1, Loop Behavior=::NewEnumerator1, Loop Duration=0.05, UseLoopDelay=false)
   &SpawnBurst_Instantaneous(Use Spawn Probability=false, UseLoopCountLimit=false)
   &SpawnBurst_Instantaneous(Spawn Time=0.001, Use Spawn Probability=false, UseLoopCountLimit=false) {enabled=false, module=SpawnBurst_Instantaneous002}
   &SpawnBurst_Instantaneous(Spawn Count=5, Spawn Group=1, Use Spawn Probability=false, UseLoopCountLimit=false) {module=SpawnBurst_Instantaneous001}
   &SpawnBurst_Instantaneous(Spawn Count=3, Spawn Group=1, Spawn Time=0.001, Use Spawn Probability=false, UseLoopCountLimit=false) {enabled=false, module=SpawnBurst_Instantaneous003}
  ]
  particleSpawnScript [
   &InitializeParticle(
    Color=&MergeHelper_ColorSwitch(
     0=<rgba 0.0469 0.0469 0.0469 1>
     1=<rgba 0.0417 0.0417 0.0417 1>
     2=?
     3=?
     Index=&SpawnGroup()
     Num=::NewEnumerator0
    )
    Lifetime=&MergeHelper_FloatSwitch(
     0=&RandomRangeFloat(Minimum=0.3, Maximum=0.35)
     1=&RandomRangeFloat(Minimum=0.3, Maximum=0.5)
     2=&RandomRangeFloat(Minimum=0.3, Maximum=1)
     3=?
     Index=&SpawnGroup()
     Num=::NewEnumerator0
    )
    Sprite Rotation Mode=::NewEnumerator2
    Sprite Size Mode=::NewEnumerator3
    Uniform Sprite Size=&Multiply_Float(
     A=&MergeHelper_FloatSwitch(
      0=&RandomRangeFloat(Minimum=60, Maximum=80)
      1=&RandomRangeFloat(Minimum=50, Maximum=70)
      2=&RandomRangeFloat(Minimum=8, Maximum=10)
      3=?
      Index=&SpawnGroup()
      Num=::NewEnumerator0
     )
     B=0.5
    )
    UsePositionOffset=false
    Write Material Random=true
   )
   /SetVariables(Particles.VisibilityTag=$Emitter.SpawnGroup) {isSetParameters=true}
   &AddVelocity(
    Rotation Angle=-30
    Rotation Axis:Vector3f=<0 0 1>
    Rotation Mode=::NewEnumerator1
    Velocity=&Multiply_VectorByFloat(
     Vector=&SelectVectorFromArray(
      Direct Array Index=&ReturnExecIndex()
      Vector Selection Array=DataInterface() {propertyValues {FloatData [<-500 0 200>, <-400 0 100>, <-300 0 50>, <-200 0 25>, <0 0 0>]}}
     )
     Float=1.25
    )
    Velocity Speed Scale=&SelectFloatFromArray(Direct Array Index=&SpawnGroup(), Float Selection Array=DataInterface() {propertyValues {FloatData [0,1]}})
   )
   &SocketLocation(SamplePosOffset=false, Socket Index=2, Socket Reader=$System.Socket Reader, Socket Sampling Mode=::NewEnumerator3, VelocitySpeedLimit=false)
   &ShapeLocation(
    MaskSpawnGroup=false
    Override Seed=false
    Shape Origin=$Output.SocketLocation.SampledPosition
    Sphere Radius=&Multiply_Float(
     A=&MergeHelper_FloatSwitch(0=0, 1=10, 2=?, 3=?, Index=&SpawnGroup(), Num=::NewEnumerator0)
     B=0.5
    )
    UseOwnerScale=true
   )
   &InheritVelocity(Inherited Velocity Speed Limit=500, Speed Limit=true, UseSpeedThreshold=false, Velocity Source=$Output.SocketLocation.SampledVelocity) {enabled=false}
   &CameraOffset(Camera Offset Amount=5) {enabled=false}
  ]
  particleUpdateScript [
   &ParticleState()
   &Drag(
    Drag=&MergeHelper_FloatSwitch(0=0, 1=11, 2=?, 3=?, Index=&SpawnGroup(), Num=::NewEnumerator0)
    Use Linear Drag=true
    Use Rotational Drag=true
   )
   &GravityForce(Gravity=&MergeHelper_VectorSwitch(0=<0 0 0>, 1=<0 0 0>, 2=?, 3=?, Index=&SpawnGroup(), Num=::NewEnumerator0))
   &SolveForcesAndVelocity(Clamp Velocity=false, Limit Acceleration=false)
   &ScaleColor(
    Scale Alpha=&MergeHelper_FloatSwitch(
     0=&FloatFromCurve(FloatCurve=DataInterface() {propertyValues {Curve {keys [<at 0.75 1>,<at 1 0>]}}})
     1=&FloatFromCurve(FloatCurve=DataInterface() {propertyValues {Curve {keys [<at 0.73 1>,<at 1 0>]}}})
     2=&FloatFromCurve(FloatCurve=DataInterface() {propertyValues {Curve {keys [<at 0 0>,<at 0.2 1>,<at 1 0>]}}})
     3=?
     Index=&SpawnGroup()
     Num=::NewEnumerator0
    )
    ScaleA=true
    ScaleRGB=true
   )
   /SetVariables(
    Particles.SmokeDyn1:Vector4f=&MakeVector4(
     X=&FloatFromCurve(FloatCurve=DataInterface() {propertyValues {Curve {defaultValue=1, keys [{tangentMode=RCTM_User, time=0, value=0}, {tangentMode=RCTM_User, time=1, value=1}]}}})
     Y=1
     Z=0
     W=0
    )
    Particles.SmokeDyn2:Vector4f=&MakeVector4(
     X=&FloatFromCurve(FloatCurve=DataInterface() {propertyValues {Curve {defaultValue=1, keys [{tangentMode=RCTM_User, time=0, value=0}, {tangentMode=RCTM_User, time=1, value=1}]}}})
     Y=&RandomRangeFloat(Minimum=0, Maximum=1)
     Z=0
     W=0
    )
   ) {isSetParameters=true}
  ]
  rendererClasses [&NiagaraSpriteRendererProperties]
  renderers [
   {rendererClass=&NiagaraSpriteRendererProperties}
   {rendererClass=&NiagaraSpriteRendererProperties, rendererIndex=1}
  ]
 }
 SmokeTrail {
  emitterUpdateScript [
   &EmitterState(Life Cycle Mode=::NewEnumerator1, Loop Behavior=::NewEnumerator1, Loop Duration=0.06, UseLoopDelay=false)
   &SpawnBurst_Instantaneous(Use Spawn Probability=false, UseLoopCountLimit=false)
   &SpawnRate(SpawnRate=600, Use Spawn Probability=false)
  ]
  particleSpawnScript [
   &InitializeParticle(
    Color=<rgba 0.0208 0.0208 0.0208 3>
    Lifetime=0.2
    Ribbon Width=40
    Ribbon Width Mode=::NewEnumerator1
    UsePositionOffset=false
    Write Material Random=true
   )
   &AddVelocity(Velocity=<0 0 100>) {enabled=false}
   &SocketLocation(SamplePosOffset=false, Socket Index=2, Socket Reader=$System.Socket Reader, Socket Sampling Mode=::NewEnumerator3, VelocitySpeedLimit=false)
  ]
  particleUpdateScript [
   &ParticleState()
   &SolveForcesAndVelocity(Clamp Velocity=false, Limit Acceleration=false)
   &ScaleColor(
    Scale Alpha=&FloatFromCurve(
     FloatCurve=DataInterface() {propertyValues {Curve {keys [<at 0 1>,<at 1 0>]}}}
     Scale Curve=&FloatFromCurve(FloatCurve=DataInterface() {propertyValues {Curve {keys [<at 0 1>,<at 0.5 0>]}}}, CurveIndex=$Emitter.Age)
    )
    ScaleA=true
    ScaleRGB=true
   )
  ]
  rendererClasses [&NiagaraRibbonRendererProperties]
  renderers [{rendererClass=&NiagaraRibbonRendererProperties}]
 }
}
summary {
 systemName=NS_Emote_GuideQuiz_BatarangTrail
 emitters [
  {emitterName=BatarangTrail, rendererClasses [&NiagaraRibbonRendererProperties]}
  {emitterName=Smoke, rendererClasses [&NiagaraSpriteRendererProperties]}
  {bEnabled=false, emitterName=BatarangTrail001, rendererClasses [&NiagaraRibbonRendererProperties]}
  {emitterName=SmokeTrail, rendererClasses [&NiagaraRibbonRendererProperties]}
 ]
 userVariables [User.RibbonIDOffset {type=NiagaraInt32}]
}
