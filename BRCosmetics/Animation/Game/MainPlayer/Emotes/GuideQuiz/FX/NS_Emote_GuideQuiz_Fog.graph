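# Niagara system for the GuideQuiz emote, layering fog, a burst and bat meshes around the emote.
# Fog spawns continuously, using spawn groups to switch shape, size and fade per layer, while InitialFog bursts outward on an age-driven radius.
# GargoyleBurst throws a short-lived sprite ring under drag and gravity; Orbit_Bats swirls meshes with a randomly signed vortex force.
# Time-of-day scalars tune fog opacity, and front-end modifiers adjust Fog spawn rates.
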NiagaraSystem /BRCosmetics/Animation/Game/MainPlayer/Emotes/GuideQuiz/FX/NS_Emote_GuideQuiz_Fog
dependencies {
 usedDataInterfaces [
  &NiagaraDataInterfaceArrayFloat3
  &NiagaraDataInterfaceCurve
  &NiagaraDataInterfaceArrayFloat
  &NiagaraDataInterfaceColorCurve
 ]
 usedDynamicInputs [
  &ColorFromCurve
  &MakeVectorFromLinearColorRGB
  &FNTimeOfDayColor
  &RandomRangeLinearColor
 ]
 usedModules [
  &SystemState
  &BlendSpaceHelper
  &self:SystemSpawnScript.SystemScriptSource.SystemScriptGraph.NiagaraNodeAssignment_0.SetVariables_#1
  &self:SystemSpawnScript.SystemScriptSource.SystemScriptGraph.NiagaraNodeAssignment_1.SetVariables_#2
 ]
}
emitters {
 Fog {
  emitterUpdateScript [
   &EmitterState(Inactive Response=::NewEnumerator1, Life Cycle Mode=::NewEnumerator1, Loop Duration=1, UseLoopDelay=false)
   &SpawnRate(
    Spawn Probability=0.5
    SpawnGroup=1
    SpawnRate=&FN_FrontEndFloatModifier(Gameplay=45, Front End=45)
    Use Spawn Probability=true
   ) {module=SpawnRate001}
   &SpawnRate(
    Spawn Probability=0.5
    SpawnRate=&FN_FrontEndFloatModifier(Gameplay=40, Front End=12)
    Use Spawn Probability=true
   )
   &SpawnRate(
    Spawn Probability=0.1
    SpawnGroup=3
    SpawnRate=&FN_FrontEndFloatModifier(Gameplay=5, Front End=5)
    Use Spawn Probability=true
   ) {module=SpawnRate003}
  ]
  particleSpawnScript [
   /SetVariables(Particles.VisibilityTag=&SpawnGroup()) {isSetParameters=true, module=SetVariables_#3}
   &InitializeParticle(
    Color=$System.Global color
    Lifetime=&MergeHelper_FloatSwitch(
     0=&RandomRangeFloat(Minimum=1.5, Maximum=2.5)
     1=&RandomRangeFloat(Minimum=2, Maximum=2.4)
     2=&RandomRangeFloat(Minimum=2, Maximum=2.4)
     3=?
     Index=&SpawnGroup()
     Num=::NewEnumerator1
    )
    Sprite Rotation Angle=&MergeHelper_FloatSwitch(
     0=0
     1=&RandomRangeFloat(Minimum=0, Maximum=360)
     2=0
     3=&RandomRangeFloat(Minimum=0, Maximum=360)
     Index=&SpawnGroup()
     Num=::NewEnumerator2
    )
    Sprite Rotation Mode=::NewEnumerator0
    Sprite Size=&MergeHelper_Vector2Switch(
     0=&Multiply_Vector2D_ByFloat(Float=&RandomRangeFloat(Minimum=25, Maximum=35), Vector2D=<1 0.5>)
     1=&RandomRangeVector2D(Minimum=<20 20>, Maximum=<40 40>)
     2=&RandomRangeVector2D(Minimum=<20 20>, Maximum=<40 40>)
     3=<10 10>
     4=?
     Index=&SpawnGroup()
     Num=::NewEnumerator2
    )
    Sprite Size Mode=::NewEnumerator4
    UsePositionOffset=false
    Write Material Random=true
   )
   &ShapeLocation(
    MaskSpawnGroup=false
    Non Uniform Scale=&MergeHelper_VectorSwitch(0=<1 1 0.5>, 1=<1 1 0.1>, 2=<1 1 0.1>, 3=<1 25 0.5>, Index=&SpawnGroup(), Num=::NewEnumerator2)
    Offset=&MergeHelper_VectorSwitch(
     0=&FN_FrontEndVector3Modifier(Gameplay=<0 0 30>, Front End=<0 0 30>)
     1=<0 0 10>
     2=<0 0 10>
     3=&SelectVectorFromArray(
      Array Sampling Mode=::NewEnumerator0
      Direct Array Index=?
      Vector Selection Array=DataInterface() {propertyValues {FloatData [<75 -25 30>, <-75 -25 30>]}}
      Random Seed=0
     )
     Index=&SpawnGroup()
     Num=::NewEnumerator2
    )
    Offset Mode=::NewEnumerator0
    Override Seed=false
    Sphere Radius=&MergeHelper_FloatSwitch(0=85, 1=90, 2=90, 3=2, Index=&SpawnGroup(), Num=::NewEnumerator2)
    UseOwnerScale=true
   )
   &AddVelocity(Velocity=&MergeHelper_VectorSwitch(
     0=<0 0 0>
     1=<0 0 0>
     2=<0 0 0>
     3=&RandomRangeVector(Minimum=<0 15 0>, Maximum=<0 25 0>)
     Index=&SpawnGroup()
     Num=::NewEnumerator2
    ))
  ]
  particleUpdateScript [
   &ParticleState()
   &ScaleSpriteSize(
    Uniform Curve Scale=10
    Uniform Curve Sprite Scale=DataInterface() {propertyValues {Curve {defaultValue=1, keys [{tangentMode=RCTM_User, time=0, value=0.7}, {tangentMode=RCTM_User, time=1, value=1}]}}}
   )
   &SubUVAnimation(
    End Frame=15
    SubUV Lookup Index:NiagaraFloat=&FloatFromCurve(FloatCurve=DataInterface() {propertyValues {Curve {defaultValue=1, keys [{tangentMode=RCTM_User, time=0, value=0}, {tangentMode=RCTM_User, time=1, value=1}]}}})
   )
   &SolveForcesAndVelocity(Clamp Velocity=false, Limit Acceleration=false)
   &ScaleSpriteSizeBySpeed(Max Scale Factor=<0 0>) {enabled=false}
   /SetVariables(
    Particles.0 - Env Fog Scale Color:LinearColor=&MakeLinearColorFromVectorAndFloat(
     "Vector (RGB)"=<1 1 1>
     "Float (Alpha)"=&FloatFromCurve(
      FloatCurve=DataInterface() {propertyValues {Curve {keys [
          {arriveTangent=3.5962, interpMode=RCIM_Cubic, leaveTangent=3.5962, tangentMode=RCTM_User, time=0, value=0}
          {
           arriveTangent=0.1817
           arriveTangentWeight=0.1842
           interpMode=RCIM_Cubic
           leaveTangent=0.1817
           leaveTangentWeight=0.7369
           tangentMode=RCTM_User
           time=0.5037
           value=0.7647
          }
          {arriveTangent=-4.7233, tangentMode=RCTM_User, time=1, value=0}
         ]}}}
      Scale Curve=&RandomRangeFloat(
       Minimum=0.02
       Maximum=&FNTimeOfDayScalar(Night Scale=0.05, Morning Scale=0.05, Day Scale=0.07, Evening Scale=0.05)
      )
     )
    )
    Particles.1 - Ground Fog Scale Color:LinearColor=&MakeLinearColorFromVectorAndFloat(
     "Vector (RGB)"=<1 1 1>
     "Float (Alpha)"=&FloatFromCurve(
      FloatCurve=DataInterface() {propertyValues {Curve {keys [
          {arriveTangent=5, interpMode=RCIM_Cubic, time=0, value=0}
          {arriveTangent=1.5181, interpMode=RCIM_Cubic, leaveTangent=1.5181, time=0.2, value=0.7}
          {
           arriveTangent=0.1817
           arriveTangentWeight=0.1842
           interpMode=RCIM_Cubic
           leaveTangent=0.1817
           leaveTangentWeight=0.7369
           tangentMode=RCTM_User
           time=0.5037
           value=0.7647
          }
          <at 1 0>
         ]}}}
      Scale Curve=&RandomRangeFloat(
       Minimum=&FNTimeOfDayScalar(Night Scale=0.035, Morning Scale=0.07, Day Scale=0.05, Evening Scale=0.07)
       Maximum=&FNTimeOfDayScalar(Night Scale=0.05, Morning Scale=0.07, Day Scale=0.07, Evening Scale=0.1)
      )
     )
    )
   ) {isSetParameters=true, module=SetVariables_#4}
   /SetVariables(
    Particles.0 - Env Fog Dyn Param:Vector4f=&MakeVector4(X=0, Y=1, Z=1, W=1)
    Particles.1 - Ground Fog Dyn Param:Vector4f=&MakeVector4(X=0, Y=1, Z=0.1, W=1)
   ) {isSetParameters=true, module=SetVariables_#5}
   &ScaleColor(
    Scale Alpha=&MergeHelper_FloatSwitch(
     0=&FloatFromCurve(
      FloatCurve=DataInterface() {propertyValues {Curve {keys [
          {arriveTangent=3.5962, interpMode=RCIM_Cubic, leaveTangent=3.5962, tangentMode=RCTM_User, time=0, value=0}
          {
           arriveTangent=0.1817
           arriveTangentWeight=0.1842
           interpMode=RCIM_Cubic
           leaveTangent=0.1817
           leaveTangentWeight=0.7369
           tangentMode=RCTM_User
           time=0.5037
           value=0.7647
          }
          {arriveTangent=-4.7233, tangentMode=RCTM_User, time=1, value=0}
         ]}}}
      Scale Curve=&RandomRangeFloat(
       Minimum=0.02
       Maximum=&FNTimeOfDayScalar(Night Scale=0.03, Morning Scale=0.03, Day Scale=0.05, Evening Scale=0.03)
      )
     )
     1=&FloatFromCurve(
      FloatCurve=DataInterface() {propertyValues {Curve {keys [
          {arriveTangent=6.4271, interpMode=RCIM_Cubic, leaveTangent=6.4271, tangentMode=RCTM_User, time=0, value=0}
          {arriveTangent=0.5834, interpMode=RCIM_Cubic, leaveTangent=0.5834, tangentMode=RCTM_User, time=0.1413, value=0.7}
          {
           arriveTangent=0.1817
           arriveTangentWeight=0.1842
           interpMode=RCIM_Cubic
           leaveTangent=0.1817
           leaveTangentWeight=0.7369
           tangentMode=RCTM_User
           time=0.5037
           value=0.8
          }
          <at 1 0>
         ]}}}
      Scale Curve=&RandomRangeFloat(
       Minimum=&FNTimeOfDayScalar(Night Scale=0.035, Morning Scale=0.03, Day Scale=0.05, Evening Scale=0.04)
       Maximum=&FNTimeOfDayScalar(Night Scale=0.05, Morning Scale=0.05, Day Scale=0.07, Evening Scale=0.07)
      )
     )
     2=&FloatFromCurve(
      FloatCurve=DataInterface() {propertyValues {Curve {keys [
          {arriveTangent=6.4271, interpMode=RCIM_Cubic, leaveTangent=6.4271, tangentMode=RCTM_User, time=0, value=0}
          {arriveTangent=0.5834, interpMode=RCIM_Cubic, leaveTangent=0.5834, tangentMode=RCTM_User, time=0.1413, value=0.7}
          {
           arriveTangent=0.1817
           arriveTangentWeight=0.1842
           interpMode=RCIM_Cubic
           leaveTangent=0.1817
           leaveTangentWeight=0.7369
           tangentMode=RCTM_User
           time=0.5037
           value=0.8
          }
          <at 1 0>
         ]}}}
      Scale Curve=&RandomRangeFloat(
       Minimum=&FNTimeOfDayScalar(Night Scale=0.035, Morning Scale=0.03, Day Scale=0.05, Evening Scale=0.04)
       Maximum=&FNTimeOfDayScalar(Night Scale=0.05, Morning Scale=0.05, Day Scale=0.07, Evening Scale=0.07)
      )
     )
     3=&FloatFromCurve(
      FloatCurve=DataInterface() {propertyValues {Curve {keys [
          {arriveTangent=3.5962, interpMode=RCIM_Cubic, leaveTangent=3.5962, tangentMode=RCTM_User, time=0, value=0}
          {
           arriveTangent=0.1817
           arriveTangentWeight=0.1842
           interpMode=RCIM_Cubic
           leaveTangent=0.1817
           leaveTangentWeight=0.7369
           tangentMode=RCTM_User
           time=0.5037
           value=0.7647
          }
          {arriveTangent=-4.7233, tangentMode=RCTM_User, time=1, value=0}
         ]}}}
      Scale Curve=&RandomRangeFloat(
       Minimum=0.02
       Maximum=&FNTimeOfDayScalar(Night Scale=0.03, Morning Scale=0.03, Day Scale=0.05, Evening Scale=0.03)
      )
     )
     Index=&SpawnGroup()
     Num=::NewEnumerator2
    )
    ScaleA=true
    ScaleRGB=true
   )
   /SetVariables(Particles.GroundAlignment:Vector3f=<0 1 0>, Particles.GroundFacing:Vector3f=<0 0 1>) {isSetParameters=true, module=SetVariables_#6}
   /SetVariables(Particles.WispCOlor:LinearColor=&MultiplyLinearColor(Linear Color=$Particles.Color, Scale Factor=&MakeVector4_FromLinearColor(LinearColor=<rgba 1 1 1 5>))) {isSetParameters=true}
  ]
  rendererClasses [&NiagaraSpriteRendererProperties]
  renderers [
   {rendererClass=&NiagaraSpriteRendererProperties}
   {rendererClass=&NiagaraSpriteRendererProperties, rendererIndex=1}
   {rendererClass=&NiagaraSpriteRendererProperties, rendererIndex=2}
   {rendererClass=&NiagaraSpriteRendererProperties, rendererIndex=3}
  ]
 }
 GargoyleBurst {
  emitterUpdateScript [
   &EmitterState(Inactive Response=::NewEnumerator1, Life Cycle Mode=::NewEnumerator1, Loop Behavior=::NewEnumerator1, Loop Duration=2.5, UseLoopDelay=false)
   &SpawnBurst_Instantaneous(Spawn Count=30, Spawn Group=1, Use Spawn Probability=false, UseLoopCountLimit=false)
   &SpawnBurst_Instantaneous(Spawn Count=20, Use Spawn Probability=false, UseLoopCountLimit=false) {module=SpawnBurst_Instantaneous001}
  ]
  particleSpawnScript [
   &InitializeParticle(
    Color=$System.Global color
    Lifetime=&RandomRangeFloat(Minimum=0.2, Maximum=0.6)
    Sprite Rotation Angle=&RandomRangeFloat(Minimum=0, Maximum=360)
    Sprite Rotation Mode=::NewEnumerator0
    Sprite Size Mode=::NewEnumerator3
    Uniform Sprite Size=&RandomRangeFloat(Minimum=10, Maximum=20)
    UsePositionOffset=false
    Write Material Random=true
   )
   &ShapeLocation(
    MaskSpawnGroup=false
    Non Uniform Scale=<1 1 0.1>
    Offset Mode=::NewEnumerator0
    Override Seed=false
    Ring Radius=25
    Shape Primitive=::NewEnumerator4
    UseOwnerScale=true
   )
   &AddVelocity(Velocity=<0 0 200>)
   &AddVelocity(Module.Use Falloff Distance=false, Override Seed=false, Velocity Mode=::NewEnumerator1, Velocity Speed=200) {module=AddVelocity001}
  ]
  particleUpdateScript [
   &ParticleState()
   &ScaleSpriteSize(
    Uniform Curve Scale=10
    Uniform Curve Sprite Scale=DataInterface() {propertyValues {Curve {defaultValue=1, keys [{tangentMode=RCTM_User, time=0, value=0.7}, {tangentMode=RCTM_User, time=1, value=1}]}}}
   )
   &SubUVAnimation(
    End Frame=15
    SubUV Animation Mode:ENiagara_SubUVLookupMode=::NewEnumerator1
    SubUV Curve=DataInterface() {propertyValues {Curve {keys [{tangentMode=RCTM_User, time=0, value=0}, {tangentMode=RCTM_User, time=1.0138, value=0.388}]}}}
   )
   &Drag(Drag=4, Use Linear Drag=true, Use Rotational Drag=true)
   &GravityForce(Gravity=<0 0 -600>)
   &SolveForcesAndVelocity(Clamp Velocity=false, Limit Acceleration=false)
   /SetVariables(
    Particles.0 - Env Fog Scale Color:LinearColor=&MakeLinearColorFromVectorAndFloat(
     "Vector (RGB)"=<1 1 1>
     "Float (Alpha)"=&FloatFromCurve(
      FloatCurve=DataInterface() {propertyValues {Curve {keys [
          {arriveTangent=3.5962, interpMode=RCIM_Cubic, leaveTangent=3.5962, tangentMode=RCTM_User, time=0, value=0}
          {
           arriveTangent=0.1817
           arriveTangentWeight=0.1842
           interpMode=RCIM_Cubic
           leaveTangent=0.1817
           leaveTangentWeight=0.7369
           tangentMode=RCTM_User
           time=0.5037
           value=0.7647
          }
          {arriveTangent=-4.7233, tangentMode=RCTM_User, time=1, value=0}
         ]}}}
      Scale Curve=&RandomRangeFloat(
       Minimum=0.02
       Maximum=&FNTimeOfDayScalar(Night Scale=0.05, Morning Scale=0.05, Day Scale=0.07, Evening Scale=0.05)
      )
     )
    )
    Particles.1 - Ground Fog Scale Color:LinearColor=&MakeLinearColorFromVectorAndFloat(
     "Vector (RGB)"=<1 1 1>
     "Float (Alpha)"=&FloatFromCurve(
      FloatCurve=DataInterface() {propertyValues {Curve {keys [
          {arriveTangent=5, interpMode=RCIM_Cubic, time=0, value=0}
          {arriveTangent=1.5181, interpMode=RCIM_Cubic, leaveTangent=1.5181, time=0.2, value=0.7}
          {
           arriveTangent=0.1817
           arriveTangentWeight=0.1842
           interpMode=RCIM_Cubic
           leaveTangent=0.1817
           leaveTangentWeight=0.7369
           tangentMode=RCTM_User
           time=0.5037
           value=0.7647
          }
          <at 1 0>
         ]}}}
      Scale Curve=&RandomRangeFloat(
       Minimum=&FNTimeOfDayScalar(Night Scale=0.035, Morning Scale=0.07, Day Scale=0.05, Evening Scale=0.07)
       Maximum=&FNTimeOfDayScalar(Night Scale=0.05, Morning Scale=0.07, Day Scale=0.07, Evening Scale=0.1)
      )
     )
    )
   ) {isSetParameters=true, module=SetVariables_#4}
   /SetVariables(
    Particles.0 - Env Fog Dyn Param:Vector4f=&MakeVector4(X=0, Y=1, Z=1, W=1)
    Particles.1 - Ground Fog Dyn Param:Vector4f=&MakeVector4(X=0, Y=1, Z=0.1, W=1)
   ) {isSetParameters=true, module=SetVariables_#5}
   &ScaleColor(
    Scale Alpha=&FloatFromCurve(
     FloatCurve=DataInterface() {propertyValues {Curve {keys [
         {arriveTangent=6.4271, interpMode=RCIM_Cubic, leaveTangent=6.4271, tangentMode=RCTM_User, time=0, value=0}
         {arriveTangent=0.5834, interpMode=RCIM_Cubic, leaveTangent=0.5834, tangentMode=RCTM_User, time=0.0598, value=0.6812}
         {
          arriveTangent=0.1817
          arriveTangentWeight=0.1842
          interpMode=RCIM_Cubic
          leaveTangent=0.1817
          leaveTangentWeight=0.7369
          tangentMode=RCTM_User
          time=0.5037
          value=0.8
         }
         <at 1 0>
        ]}}}
     Scale Curve=&RandomRangeFloat(
      Minimum=&FNTimeOfDayScalar(Night Scale=0.035, Morning Scale=0.03, Day Scale=0.05, Evening Scale=0.04)
      Maximum=&FNTimeOfDayScalar(Night Scale=0.05, Morning Scale=0.05, Day Scale=0.07, Evening Scale=0.07)
     )
    )
    ScaleA=true
    ScaleRGB=true
   )
  ]
  rendererClasses [&NiagaraSpriteRendererProperties]
  renderers [{rendererClass=&NiagaraSpriteRendererProperties}]
 }
 InitialFog {
  emitterUpdateScript [
   &EmitterState(Inactive Response=::NewEnumerator1, Life Cycle Mode=::NewEnumerator1, Loop Behavior=::NewEnumerator1, Loop Duration=2.5, UseLoopDelay=false)
   &SpawnBurst_Instantaneous(Spawn Count=15, Spawn Group=1, Use Spawn Probability=false, UseLoopCountLimit=false)
   &SpawnBurst_Instantaneous(Spawn Count=10, Use Spawn Probability=false, UseLoopCountLimit=false) {module=SpawnBurst_Instantaneous001}
  ]
  particleSpawnScript [
   /SetVariables(Particles.VisibilityTag=&SpawnGroup()) {isSetParameters=true, module=SetVariables_#3}
   &InitializeParticle(
    Color=$System.Global color
    Lifetime=&MergeHelper_FloatSwitch(
     0=&RandomRangeFloat(Minimum=1.5, Maximum=2.5)
     1=&RandomRangeFloat(Minimum=2, Maximum=2.4)
     2=?
     3=?
     Index=&SpawnGroup()
     Num=::NewEnumerator0
    )
    Sprite Rotation Angle=&MaskFloat(
     Mask Source=0
     Mask Parameter=?
     Mask=0
     Masked Float=0
     Passthrough Float=&RandomRangeFloat(Minimum=0, Maximum=360)
    )
    Sprite Rotation Mode=::NewEnumerator0
    Sprite Size=&MergeHelper_Vector2Switch(
     0=&Multiply_Vector2D_ByFloat(Float=&RandomRangeFloat(Minimum=25, Maximum=35), Vector2D=<1 0.5>)
     1=&RandomRangeVector2D(Minimum=<20 20>, Maximum=<40 40>)
     2=?
     3=?
     4=?
     Index=&SpawnGroup()
     Num=::NewEnumerator0
    )
    Sprite Size Mode=::NewEnumerator4
    UsePositionOffset=false
    Write Material Random=true
   )
  ]
  particleUpdateScript [
   &ParticleState()
   &ShapeLocation(
    MaskSpawnGroup=false
    Non Uniform Scale=&MergeHelper_VectorSwitch(0=<1 1 0.5>, 1=<1 1 0.1>, 2=?, 3=?, Index=&SpawnGroup(), Num=::NewEnumerator0)
    Offset=&MergeHelper_VectorSwitch(
     0=&FN_FrontEndVector3Modifier(Gameplay=<0 0 30>, Front End=<0 0 30>)
     1=<0 0 10>
     2=?
     3=?
     Index=&SpawnGroup()
     Num=::NewEnumerator0
    )
    Offset Mode=::NewEnumerator0
    Override Seed=false
    Sphere Radius=&FloatFromCurve(
     FloatCurve=DataInterface() {propertyValues {Curve {
        defaultValue=1
        keys [
         {arriveTangent=3.3483, interpMode=RCIM_Cubic, leaveTangent=3.3483, tangentMode=RCTM_User, time=0, value=0}
         {interpMode=RCIM_Cubic, tangentMode=RCTM_User, time=1, value=1}
        ]
       }}}
     CurveIndex=&Subtract_Float(A=$Emitter.Age, B=0)
     Scale Curve=&MergeHelper_FloatSwitch(0=85, 1=90, 2=?, 3=?, Index=&SpawnGroup(), Num=::NewEnumerator0)
    )
    UseOwnerScale=true
   )
   &ScaleSpriteSize(
    Uniform Curve Scale=10
    Uniform Curve Sprite Scale=DataInterface() {propertyValues {Curve {defaultValue=1, keys [{tangentMode=RCTM_User, time=0, value=0.7}, {tangentMode=RCTM_User, time=1, value=1}]}}}
   )
   &SubUVAnimation(
    End Frame=15
    SubUV Animation Mode:ENiagara_SubUVLookupMode=::NewEnumerator1
    SubUV Curve=DataInterface() {propertyValues {Curve {keys [{tangentMode=RCTM_User, time=0, value=0}, {tangentMode=RCTM_User, time=1.0138, value=0.388}]}}}
   )
   &SolveForcesAndVelocity(Clamp Velocity=false, Limit Acceleration=false)
   /SetVariables(
    Particles.0 - Env Fog Scale Color:LinearColor=&MakeLinearColorFromVectorAndFloat(
     "Vector (RGB)"=<1 1 1>
     "Float (Alpha)"=&FloatFromCurve(
      FloatCurve=DataInterface() {propertyValues {Curve {keys [
          {arriveTangent=3.5962, interpMode=RCIM_Cubic, leaveTangent=3.5962, tangentMode=RCTM_User, time=0, value=0}
          {
           arriveTangent=0.1817
           arriveTangentWeight=0.1842
           interpMode=RCIM_Cubic
           leaveTangent=0.1817
           leaveTangentWeight=0.7369
           tangentMode=RCTM_User
           time=0.5037
           value=0.7647
          }
          {arriveTangent=-4.7233, tangentMode=RCTM_User, time=1, value=0}
         ]}}}
      Scale Curve=&RandomRangeFloat(
       Minimum=0.02
       Maximum=&FNTimeOfDayScalar(Night Scale=0.05, Morning Scale=0.05, Day Scale=0.07, Evening Scale=0.05)
      )
     )
    )
    Particles.1 - Ground Fog Scale Color:LinearColor=&MakeLinearColorFromVectorAndFloat(
     "Vector (RGB)"=<1 1 1>
     "Float (Alpha)"=&FloatFromCurve(
      FloatCurve=DataInterface() {propertyValues {Curve {keys [
          {arriveTangent=5, interpMode=RCIM_Cubic, time=0, value=0}
          {arriveTangent=1.5181, interpMode=RCIM_Cubic, leaveTangent=1.5181, time=0.2, value=0.7}
          {
           arriveTangent=0.1817
           arriveTangentWeight=0.1842
           interpMode=RCIM_Cubic
           leaveTangent=0.1817
           leaveTangentWeight=0.7369
           tangentMode=RCTM_User
           time=0.5037
           value=0.7647
          }
          <at 1 0>
         ]}}}
      Scale Curve=&RandomRangeFloat(
       Minimum=&FNTimeOfDayScalar(Night Scale=0.035, Morning Scale=0.07, Day Scale=0.05, Evening Scale=0.07)
       Maximum=&FNTimeOfDayScalar(Night Scale=0.05, Morning Scale=0.07, Day Scale=0.07, Evening Scale=0.1)
      )
     )
    )
   ) {isSetParameters=true, module=SetVariables_#4}
   /SetVariables(
    Particles.0 - Env Fog Dyn Param:Vector4f=&MakeVector4(X=0, Y=1, Z=1, W=1)
    Particles.1 - Ground Fog Dyn Param:Vector4f=&MakeVector4(X=0, Y=1, Z=0.1, W=1)
   ) {isSetParameters=true, module=SetVariables_#5}
   &ScaleColor(
    Scale Alpha=&MergeHelper_FloatSwitch(
     0=&FloatFromCurve(
      FloatCurve=DataInterface() {propertyValues {Curve {keys [
          {arriveTangent=8.476, interpMode=RCIM_Cubic, leaveTangent=8.476, tangentMode=RCTM_User, time=0, value=0}
          {arriveTangent=0.8037, interpMode=RCIM_Cubic, leaveTangent=0.8037, tangentMode=RCTM_User, time=0.0701, value=0.6842}
          {
           arriveTangent=0.1817
           arriveTangentWeight=0.1842
           interpMode=RCIM_Cubic
           leaveTangent=0.1817
           leaveTangentWeight=0.7369
           tangentMode=RCTM_User
           time=0.5037
           value=0.7647
          }
          {arriveTangent=-4.7233, tangentMode=RCTM_User, time=1, value=0}
         ]}}}
      Scale Curve=&RandomRangeFloat(
       Minimum=0.02
       Maximum=&FNTimeOfDayScalar(Night Scale=0.03, Morning Scale=0.03, Day Scale=0.05, Evening Scale=0.03)
      )
     )
     1=&FloatFromCurve(
      FloatCurve=DataInterface() {propertyValues {Curve {keys [
          {arriveTangent=6.4271, interpMode=RCIM_Cubic, leaveTangent=6.4271, tangentMode=RCTM_User, time=0, value=0}
          {arriveTangent=0.5834, interpMode=RCIM_Cubic, leaveTangent=0.5834, tangentMode=RCTM_User, time=0.0598, value=0.6812}
          {
           arriveTangent=0.1817
           arriveTangentWeight=0.1842
           interpMode=RCIM_Cubic
           leaveTangent=0.1817
           leaveTangentWeight=0.7369
           tangentMode=RCTM_User
           time=0.5037
           value=0.8
          }
          <at 1 0>
         ]}}}
      Scale Curve=&RandomRangeFloat(
       Minimum=&FNTimeOfDayScalar(Night Scale=0.035, Morning Scale=0.03, Day Scale=0.05, Evening Scale=0.04)
       Maximum=&FNTimeOfDayScalar(Night Scale=0.05, Morning Scale=0.05, Day Scale=0.07, Evening Scale=0.07)
      )
     )
     2=?
     3=?
     Index=&SpawnGroup()
     Num=::NewEnumerator0
    )
    ScaleA=true
    ScaleRGB=true
   )
   /SetVariables(Particles.GroundAlignment:Vector3f=<0 1 0>, Particles.GroundFacing:Vector3f=<0 0 1>) {isSetParameters=true, module=SetVariables_#6}
  ]
  rendererClasses [&NiagaraSpriteRendererProperties]
  renderers [
   {rendererClass=&NiagaraSpriteRendererProperties}
   {rendererClass=&NiagaraSpriteRendererProperties, rendererIndex=1}
   {rendererClass=&NiagaraSpriteRendererProperties, rendererIndex=2}
  ]
 }
 Orbit_Bats {
  emitterUpdateScript [
   &EmitterState(Inactive Response=::NewEnumerator1, Life Cycle Mode=::NewEnumerator1, Loop Behavior=::NewEnumerator1, Loop Duration=1, UseLoopDelay=false)
   &SpawnRate(
    SpawnRate=&FloatFromCurve(
     FloatCurve=DataInterface() {propertyValues {Curve {keys [<at 0.4266 1>,<at 0.531 0>]}}}
     CurveIndex=$Emitter.NormalizedLoopAge
     Scale Curve=10
    )
    Use Spawn Probability=false
   )
   &SpawnBurst_Instantaneous(
    Spawn Count=&RandomRangeInt(Minimum=5, Maximum=15)
    Use Spawn Probability=false
    UseLoopCountLimit=false
   )
  ]
  particleSpawnScript [
   /Niagara/Modules/Spawn/Initialization/InitializeParticle(
    Lifetime=&UniformRangedFloat(Minimum=0.5, Maximum=1.75)
    Mesh Scale=&VectorFromFloat(Value=&RandomRangeFloat(Minimum=0.5, Maximum=1.8))
    Module.Write Color=true
    Module.Write Lifetime=true
    Module.Write Mass=false
    Module.Write Material Random=true
    Module.Write Position=false
    Module.Write Scale=true
    Module.Write SpriteRotation=false
    Module.Write SpriteSize=false
    Position=&SimulationPosition()
   ) {enabled=false}
   /Niagara/Modules/Spawn/Initialization/V2/InitializeParticle(
    Lifetime=&UniformRangedFloat(Minimum=0.5, Maximum=0.75)
    Mesh Scale Mode=::NewEnumerator3
    Mesh Uniform Scale=&RandomRangeFloat(Minimum=0.5, Maximum=1)
    UsePositionOffset=false
    Write Material Random=true
   ) {module=InitializeParticle001}
   &ShapeLocation(Hemisphere Distribution=<1 0.4>, MaskSpawnGroup=false, Override Seed=false, Sphere Radius=25, UseOwnerScale=true)
   &DynamicMaterialParameters(
    Index 0 Param 1=&RandomRangeFloat(Minimum=0, Maximum=3)
    Index 0 Param 2=&RandomRangeFloat(Minimum=1, Maximum=1.2)
    Module.Param0WriteEnabled=true
    Module.Param1WriteEnabled=true
    Module.Param2WriteEnabled=true
    Module.Param3WriteEnabled=true
   )
   /Game/Effects/Niagara/Modules/Orbit/FN_Orbit_Init_Variables_v2() {enabled=false}
   &InitialMeshOrientation(
    Rotation=&RandomRangeVector(Minimum=<0 0 0>, Maximum=<1 1 0>)
    Use Orientation Vector=true
    Use Rotation Vector=true
   ) {enabled=false}
   &AddVelocity(
    Cone Angle=90
    Cone Axis=<0 0 1>
    Module.UseConeFalloff=false
    Module.UseVelDistribution=false
    Override Seed=false
    Velocity Mode=::NewEnumerator2
    Velocity Speed=&RandomRangeFloat(Minimum=600, Maximum=800)
   )
  ]
  particleUpdateScript [
   &ParticleState()
   &ScaleMeshSize(Scale Factor=&VectorFromFloat(Value=&FloatFromCurve(FloatCurve=DataInterface() {propertyValues {Curve {keys [{arriveTangent=0.4267, interpMode=RCIM_Cubic, leaveTangent=0.4267, tangentMode=RCTM_User, time=0, value=0.5}, <at 0.5 1>]}}})))
   &ScaleColor(
    Scale Alpha=&FloatFromCurve(FloatCurve=DataInterface() {propertyValues {Curve {keys [
         {arriveTangent=0.7333, interpMode=RCIM_Cubic, time=0.1, value=0.5}
         {arriveTangent=-0.5556, interpMode=RCIM_Cubic, leaveTangent=-0.5556, time=0.75, value=0.55}
         <at 1 0>
        ]}}})
    ScaleA=true
    ScaleRGB=true
   )
   /Game/Effects/Niagara/Modules/Orbit/FN_Orbit_v4(
    Height min/max:Vector2f=<-200 200>
    Radius min/max:Vector2f=<150 375>
    Rotation Offset min/max:Vector2f=<0 360>
    Rotation Rate min/max:Vector2f=<0.3 0.3>
   ) {enabled=false}
   &VortexForce(
    UseOriginOffset=false
    Vortex Force Amount=&SelectFloatFromArray(
     Array Sampling Mode=::NewEnumerator0
     Direct Array Index=?
     Float Selection Array=DataInterface() {propertyValues {FloatData [2000,-2000]}}
     Random Seed=0
    )
   )
   &Drag(Drag=3, Use Linear Drag=true, Use Rotational Drag=true)
   &SolveForcesAndVelocity(Clamp Velocity=false, Limit Acceleration=false)
  ]
  rendererClasses [&NiagaraMeshRendererProperties]
  renderers [{rendererClass=&NiagaraMeshRendererProperties, rendererIndex=0}]
 }
 SmokeLWE {rendererClasses [], renderers []}
}
summary {
 systemName=NS_Emote_GuideQuiz_Fog
 emitters [
  {emitterName=Fog, rendererClasses [&NiagaraSpriteRendererProperties]}
  {emitterName=InitialFog, rendererClasses [&NiagaraSpriteRendererProperties]}
  {emitterName=GargoyleBurst, rendererClasses [&NiagaraSpriteRendererProperties]}
  {emitterName=Orbit_Bats, rendererClasses [&NiagaraMeshRendererProperties]}
  {emitterName=SmokeLWE, rendererClasses []}
 ]
 userVariables [
  User.bIsFrontEnd {type=NiagaraBool}
  User.bIsFrontEndPreview {type=NiagaraBool}
 ]
}
